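# Physics asset for the M_MED_Vivid_Trooper male medium cosmetic body, set up for secondary motion on its belt.
# Small capsules on four dyn_belt bones form a constraint chain rooted at dyn_belt_MSTR.
# Swing limits widen from the root toward the tip, while twist stays tight throughout.
# A large capsule on the right upper thigh twist bone has no constraint, apparently serving as a collider for the belt.
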PhysicsAsset /BRCosmetics/Characters/Player/Male/Medium/Bodies/M_MED_Vivid_Trooper/Meshes/M_MED_Vivid_Trooper_Physics
bodies {
 dyn_belt_01 [{shapeType=Capsule, length=2.87, radius=1.625, shapeName=dyn_belt_01_capsule, center=<2.8009 0 0>, rotation=<rot 90 0 0>}]
 dyn_belt_02 [{shapeType=Capsule, length=2.8756, radius=1.5, shapeName=dyn_belt_02_capsule, center=<2.8009 0 0>, rotation=<rot 90 0 0>}]
 dyn_belt_03 [{shapeType=Capsule, length=2.75, radius=1.5, shapeName=dyn_belt_03_capsule, center=<2.8009 0 0>, rotation=<rot 90 0 0>}]
 dyn_belt_04 [{shapeType=Capsule, length=2.625, radius=1.625, shapeName=dyn_belt_04_capsule, center=<2.8009 0 0>, rotation=<rot 90 0 0>}]
 dyn_belt_MSTR [{shapeType=Capsule, length=0, radius=0.505, shapeName=dyn_belt_MSTR_capsule, rotation=<rot 0 -90 0>}]
 thigh_twist_01_r [{shapeType=Capsule, length=30.125, radius=10.25, shapeName=thigh_twist_01_r_capsule, center=<-7.906 0 0>, rotation=<rot 90 0 0>}]
}
constraints [
 {bone1Name=dyn_belt_01, bone2Name=dyn_belt_MSTR, swing1LimitDegrees=20, swing2LimitDegrees=20, twistLimitDegrees=5}
 {bone1Name=dyn_belt_02, bone2Name=dyn_belt_01, swing1LimitDegrees=35, swing2LimitDegrees=35, twistLimitDegrees=5}
 {bone1Name=dyn_belt_03, bone2Name=dyn_belt_02, swing1LimitDegrees=45, swing2LimitDegrees=45, twistLimitDegrees=5}
 {bone1Name=dyn_belt_04, bone2Name=dyn_belt_03, swing1LimitDegrees=45, swing2LimitDegrees=45, twistLimitDegrees=5}
]
