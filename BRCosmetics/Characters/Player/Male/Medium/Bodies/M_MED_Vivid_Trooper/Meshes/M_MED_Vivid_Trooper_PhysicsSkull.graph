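# Physics asset for the skull piece on the male medium Vivid Trooper body, covering skull and mandible motion, likely as secondary dynamics.
# Capsule bodies cover the skull, both mandibles, a tiny master bone, spine_05 and the left upper-arm twist bone.
# Skull and mandibles each hang off dyn_skull_MSTR with tight, per-axis asymmetric angular limits.
# The master bone and left arm connect to spine_05 with free swing and twist.
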PhysicsAsset /BRCosmetics/Characters/Player/Male/Medium/Bodies/M_MED_Vivid_Trooper/Meshes/M_MED_Vivid_Trooper_PhysicsSkull
bodies {
 dyn_mandible_l [{
   shapeType=Capsule
   length=18.3037
   radius=4.8216
   shapeName=dyn_mandible_l_capsule
   center=<11.7815 -0.6658 -5.5876>
   rotation=<rot 1.6622 -89.8673 -73.8006>
  }]
 dyn_mandible_r [{
   shapeType=Capsule
   length=18.6988
   radius=4.521
   shapeName=dyn_mandible_r_capsule
   center=<15.0524 0.1736 -7.0028>
   rotation=<rot -2.1245 -98.8882 -77.1898>
  }]
 dyn_skull [{shapeType=Capsule, length=5, radius=9.4173, shapeName=dyn_skull_capsule, center=<15.7184 -0.701 3.1597>, rotation=<rot 0 -91.3993 -69.7206>}]
 dyn_skull_MSTR [{shapeType=Capsule, length=0.505, radius=0.505, shapeName=dyn_skull_MSTR_capsule, rotation=<rot 0 -90 0>}]
 spine_05 [{shapeType=Capsule, length=6.8291, radius=12.6166, shapeName=spine_05_capsule, center=<-5.5842 0 0>}]
 upperarm_twist_01_l [{
   shapeType=Capsule
   length=12.6237
   radius=4.8382
   shapeName=upperarm_twist_01_l_capsule
   center=<10.8448 0.5373 1.0907>
   rotation=<rot 83.2417 0 0>
  }]
}
constraints [
 {bone1Name=dyn_skull, bone2Name=dyn_skull_MSTR, swing1LimitDegrees=1, swing2LimitDegrees=20, twistLimitDegrees=1}
 {bone1Name=dyn_mandible_r, bone2Name=dyn_skull_MSTR, swing1LimitDegrees=10, swing2LimitDegrees=1, twistLimitDegrees=15}
 {bone1Name=dyn_mandible_l, bone2Name=dyn_skull_MSTR, swing1LimitDegrees=15, swing2LimitDegrees=10, twistLimitDegrees=1}
 {
  bone1Name=upperarm_twist_01_l
  bone2Name=spine_05
  swing1LimitDegrees=45
  swing1Motion=Free
  swing2LimitDegrees=45
  swing2Motion=Free
  twistLimitDegrees=45
  twistMotion=Free
 }
 {
  bone1Name=dyn_skull_MSTR
  bone2Name=spine_05
  swing1LimitDegrees=45
  swing1Motion=Free
  swing2LimitDegrees=45
  swing2Motion=Free
  twistLimitDegrees=45
  twistMotion=Free
 }
]
