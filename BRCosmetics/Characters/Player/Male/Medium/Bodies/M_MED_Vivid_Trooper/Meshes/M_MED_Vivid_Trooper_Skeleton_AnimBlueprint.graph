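# Cosmetic animation blueprint for the Vivid Trooper body, copying pose and curves from the attached parent mesh and adding secondary physics motion.
# The cached pose feeds Skydive and Default Control Rigs, blended by bIsParachuteOpened, with Skydive active when true.
# An additive dyn_skull_MSTR transform with neutral values precedes two chained RigidBody simulations.
# These use separate physics assets, simulated relative to root and then the skull bone, with scaled component motion.
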AnimBlueprint /BRCosmetics/Characters/Player/Male/Medium/Bodies/M_MED_Vivid_Trooper/Meshes/M_MED_Vivid_Trooper_Skeleton_AnimBlueprint
target_skeleton=&/M_MED_Vivid_Trooper_Skeleton
graphs [
 AnimGraph {nodes [
   n1 = Control Rig(Source=Use cached pose 'cachedPose'()) {
    alphaBoolBlend {blendOption=Linear}
    controlRigAssetReference {blueprintRigClass=&/M_MED_Vivid_Trooper_CtrlRigSkydive.M_MED_Vivid_Trooper_CtrlRigSkydive_C}
   }
   n2 = Control Rig(Source=Use cached pose 'cachedPose'()) {
    alphaBoolBlend {blendOption=Linear}
    controlRigAssetReference {blueprintRigClass=&/M_MED_Vivid_Trooper_CtrlRigDefault.M_MED_Vivid_Trooper_CtrlRigDefault_C}
   }
   n3 = Get bIsParachuteOpened()
   n4 = "Transform (Modify) Bone - Bone: dyn_skull_MSTR"(
    ComponentPose=Local To Component(LocalPose=Blend Poses by bool(BlendPose_0=@n1, BlendPose_1=@n2, BlendTime_0=1, BlendTime_1=1, bActiveValue=@n3) {blendType=Linear})
    Alpha=1
    Translation="0,0,0"
    Rotation="0,0,0"
    Scale="1,1,1"
   ) {
    rotationMode=BMM_Additive
    scaleMode=BMM_Additive
    translationMode=BMM_Additive
    alphaBoolBlend {blendOption=Linear}
    boneToModify {boneName=dyn_skull_MSTR}
   }
   n5 = RigidBody(ComponentPose=@n4) {
    bOverrideWorldGravity=true
    overridePhysicsAsset=&/M_MED_Vivid_Trooper_Physics
    simulationSpace=BaseBoneSpace
    alphaBoolBlend {blendOption=Linear}
    baseBoneRef {boneName=root}
    componentAppliedLinearAccClamp=<3000 3000 3000>
    componentLinearAccScale=<0.2 0.2 0.4>
    componentLinearVelScale=<0.1 0.1 0.2>
    overrideWorldGravity {z=-980}
   }
   n6 = RigidBody(ComponentPose=@n5) {
    bOverrideWorldGravity=true
    overridePhysicsAsset=&/M_MED_Vivid_Trooper_PhysicsSkull
    simulationSpace=BaseBoneSpace
    alphaBoolBlend {blendOption=Linear}
    baseBoneRef {boneName=dyn_skull_MSTR}
    componentAppliedLinearAccClamp=<3000 3000 3000>
    componentLinearAccScale=<0.2 0.2 0.4>
    componentLinearVelScale=<0.1 0.1 0.1>
    overrideWorldGravity {z=-980}
   }
   Output Pose(Result=Component To Local(ComponentPose=@n6)) {layerGroup=None}
   Save cached pose 'cachedPose'(Pose=Copy Pose From Mesh() {bCopyCurves=true, bUseAttachedParent=true}) {cacheName=cachedPose}
  ]}
 EventGraph {nodes [
   TryGetPawnOwner()
   Event BlueprintUpdateAnimation()
  ]}
]
variables [parachuteAlpha {type=number}]
